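# Masked, two-sided lit material for the Raven critter character's feather particles, usable on Niagara sprites and meshes.
# An offset, power-shaped vertical UV gradient blends particle color toward its hue-shifted, desaturated version, or blends two parameter colors, chosen by a static switch.
# The gradient tints a reused quest chicken feather texture, with matching normal and mask maps.
# Another static switch picks particle or texture alpha.
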Material /Game/Characters/Player/Male/Medium/Bodies/M_MED_Critter_Raven/FX/M_CritterRaven_ParticleFeathers
graph {
 properties {
  bUsedWithNiagaraMeshParticles=true
  bUsedWithNiagaraSprites=true
  blendMode=BLEND_Masked
  pixelDepthOffsetMode=PDOM_Legacy
  twoSided=true
  usedShadingModels=Default Lit
 }
 nodes [
  n1 = ParticleColor()
  n2 = Desaturation(
   None=MaterialFunctionCall(Hue Shift Percentage=ScalarParameter(), Texture=@n1.RGB) {materialFunction=&HueShift}.Result
   Fraction=ScalarParameter() {parameterName=DesaturationAmount}
  )
  n3 = Add(A=TextureCoordinate(), B=LinearInterpolate(Alpha=ScalarParameter() {defaultValue=0.5, parameterName=GradientOffset}) {constA=-1})
  n4 = Saturate(None=Power(Base=ComponentMask(None=@n3) {g=true}, Exp=ScalarParameter() {defaultValue=1, parameterName=GradientMaskPower}))
  n5 = VectorParameter() {parameterName=TopGradientColor, defaultValue=<rgba 0.6354 0 0.0197 1>}
  n6 = VectorParameter() {parameterName=BottomGradientColor, defaultValue=<rgba 0 0.6322 0.6354 1>}
  n7 = TextureSample() {texture=/Game/Environments/Apollo/Sets/Quest/Props/Textures/T_Quest_Chicken_Feather_01_D}
  Multiply(
   A=StaticSwitchParameter(True=LinearInterpolate(A=@n1.RGB, B=@n2, Alpha=@n4), False=LinearInterpolate(A=@n5.RGB, B=@n6.RGB, Alpha=@n4)) {parameterName=UseParticleColor}
   B=@n7.RGB
  )
  StaticSwitchParameter(True=@n1.A, False=@n7.A)
  TextureSample() {samplerType=SAMPLERTYPE_Normal, texture=/Game/Environments/Apollo/Sets/Quest/Props/Textures/T_Quest_Chicken_Feather_01_N}
  TextureSample() {samplerType=SAMPLERTYPE_Masks, texture=/Game/Environments/Apollo/Sets/Quest/Props/Textures/T_Quest_Chicken_Feather_01_S}
 ]
}
